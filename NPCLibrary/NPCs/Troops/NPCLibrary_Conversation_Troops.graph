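# Conversation database for the Troops NPC, entered through the Conversation.Entry.Troops tag.
# A single speech node shows a default greeting, with contextual lines that replace it when the player wears particular character cosmetics, including Troops' own skin.
# Cosmetic checks use HasCID controller requirements.
# The speech node then branches through link nodes into the shared NPC library quest and basic services conversations.
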ConversationDatabase /NPCLibrary/NPCs/Troops/NPCLibrary_Conversation_Troops
entry_points [{tag=Conversation.Entry.Troops, destinations [#1]}]
nodes [
 ConversationEntryPointNode_7 {
  id=#1
  class=ConversationEntryPointNode
  outputs [#2]
  properties {entryTag {tagName=Conversation.Entry.Troops}}
 }
 Speech_C_4 {
  id=#2
  class=Speech_C
  outputs [#3, #4]
  properties {
   nodeName=""
   generalConfig {
    defaultMessage=Stay out of the fruit cellar.
    contextualMessages [
     {
      message=You look like one of them...
      requirementMatchPolicy=RequireAny
      contextRequirements [{requirement {
         class=FortControllerRequirement_HasCID
         properties {allowedCIds [
           /BRCosmetics/Athena/Items/Cosmetics/Characters/CID_254_Athena_Commando_M_Zombie
           /BRCosmetics/Athena/Items/Cosmetics/Characters/CID_246_Athena_Commando_F_Grave
           /BRCosmetics/Athena/Items/Cosmetics/Characters/CID_030_Athena_Commando_M_Halloween
           /BRCosmetics/Athena/Items/Cosmetics/Characters/CID_917_Athena_Commando_M_DurrburgerSkull
          ]}
        }}]
      priority {value=0, curve {curveTable=null, rowName=None}}
     }
     {
      message=What's got you down?
      contextRequirements [{requirement {class=FortControllerRequirement_HasCID, properties {allowedCIds [&Character_Bites]}}}]
      priority {value=0, curve {curveTable=null, rowName=None}}
     }
     {
      message="Yeah, I don't watch a lot of television."
      contextRequirements [{requirement {class=FortControllerRequirement_HasCID, properties {allowedCIds [&Character_PinkSpike]}}}]
      priority {value=0, curve {curveTable=null, rowName=None}}
     }
     {
      message=What kind of music you listen to?
      contextRequirements [{requirement {class=FortControllerRequirement_HasCID, properties {allowedCIds [&Character_ChillCat]}}}]
      priority {value=0, curve {curveTable=null, rowName=None}}
     }
     {
      message=Are you...a bear inside another bear?
      contextRequirements [{requirement {class=FortControllerRequirement_HasCID, properties {allowedCIds [&Character_BadBear]}}}]
      priority {value=0, curve {curveTable=null, rowName=None}}
     }
     {
      message=You look like a flower.
      contextRequirements [{requirement {class=FortControllerRequirement_HasCID, properties {allowedCIds [&Character_TheHerald]}}}]
      priority {value=0, curve {curveTable=null, rowName=None}}
     }
     {
      message=You remind me of this wrestler I knew...
      contextRequirements [{requirement {class=FortControllerRequirement_HasCID, properties {allowedCIds [&Character_RoseDust]}}}]
      priority {value=0, curve {curveTable=null, rowName=None}}
     }
     {
      message="Oh boy, not you again..."
      contextRequirements [{requirement {class=FortControllerRequirement_HasCID, properties {allowedCIds [/BRCosmetics/Athena/Items/Cosmetics/Characters/Character_Troops]}}}]
      priority {value=0, curve {curveTable=null, rowName=None}}
     }
    ]
   }
  }
 }
 ConversationLinkNode_0 {id=#4, class=ConversationLinkNode, properties {remoteEntryTag {tagName=Conversation.Entry.BasicServices}}}
 ConversationLinkNode_1 {id=#3, class=ConversationLinkNode, properties {remoteEntryTag {tagName=Conversation.Entry.NPCLibrary.Quests}}}
]
